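# Animation Blueprint for the FNCS S20 backpack cosmetic, procedurally posing the pack and its laurels.
# Each update, time-based sines bob the laurels and ripple phase-offset rotations along mirrored bone chains, while owner spine bend drives pack offsets.
# A chain of Modify Bone nodes applies these plus fixed pack adjustments and a skydiving correction.
# Events also set jump/fall and CarrotCake body-mesh alphas.
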AnimBlueprint /BRCosmetics/Accessories/FORT_Backpacks/Backpack_F_MED_FNCS_S20/Meshes/F_MED_FNCS_S20_Pack_AnimBP
target_skeleton=&/F_MED_FNCS_S20_Pack_Skeleton
graphs [
 AnimGraph {nodes [
   "Transform (Modify) Bone - Bone: pack"(Translation="0,-0.1,4", Rotation="0,0,0") {translationMode=BMM_Additive, alphaBoolBlend {blendOption=Linear}, boneToModify {boneName=pack}, translation {y=-0.1, z=4}}
   n1 = "Transform (Modify) Bone - Bone: pack"!(
    ComponentPose=Local To Component(LocalPose=Local Space Ref Pose())
    Alpha
    Translation=Get backpack_translation()
    Rotation=Get backpack_rotate()
   ) -> (Pose) {translationMode=BMM_Additive, translationSpace=BCS_BoneSpace, alphaBoolBlend {blendOption=Linear}, boneToModify {boneName=pack}}
   n2 = "Transform (Modify) Bone - Bone: pack"!(ComponentPose=@n1, Alpha, Translation="0,-2,4", Rotation="0,0,0", Scale:Vector) -> (Pose) {translationMode=BMM_Additive, alphaBoolBlend {blendOption=Linear}, boneToModify {boneName=pack}, translation {y=-2, z=4}}
   n3 = "Transform (Modify) Bone - Bone: pack"(ComponentPose=@n2, Translation="0,-1,0", Rotation="0,0,15") {
    rotationMode=BMM_Additive
    rotationSpace=BCS_BoneSpace
    translationMode=BMM_Additive
    alphaBoolBlend {blendOption=Linear}
    alphaScaleBiasClamp {bInterpResult=true}
    boneToModify {boneName=pack}
    rotation {roll=15}
    translation {y=-1}
   }
   n4 = "Transform (Modify) Bone - Bone: pack"(ComponentPose=@n3, Translation="0,0,0", Rotation="0,0,10") {
    rotationMode=BMM_Additive
    rotationSpace=BCS_BoneSpace
    translationMode=BMM_Additive
    alphaBoolBlend {blendOption=Linear}
    alphaScaleBiasClamp {bInterpResult=true}
    boneToModify {boneName=pack}
    rotation {roll=10}
   }
   n5 = "Transform (Modify) Bone - Bone: laures"(ComponentPose=@n4, Alpha=1, Translation=Get laurels_movement() -> (laurels_movement:Vector), Rotation="0,0,0", Scale="1,1,1") {translationMode=BMM_Additive, translationSpace=BCS_BoneSpace, alphaBoolBlend {blendOption=Linear}, boneToModify {boneName=laures}}
   n6 = Get laurel_rotation_1() -> (laurel_rotation_1:double)
   n7 = "Transform (Modify) Bone - Bone: laurel_1_r"!(ComponentPose=@n5, Alpha=1, Translation="0,0,0", Rotation_Roll:float=0, Rotation_Pitch:float=0, Rotation_Yaw:float=@n6, Scale="1,1,1") -> (Pose) {rotationMode=BMM_Additive, rotationSpace=BCS_BoneSpace, alphaBoolBlend {blendOption=Linear}, boneToModify {boneName=laurel_1_r}}
   n8 = Get laurel_rotation_2() -> (laurel_rotation_2:double)
   n9 = "Transform (Modify) Bone - Bone: laurel_2_r"!(ComponentPose=@n7, Alpha=1, Translation="0,0,0", Rotation_Roll:float=0, Rotation_Pitch:float=0, Rotation_Yaw:float=@n8, Scale="1,1,1") -> (Pose) {rotationMode=BMM_Additive, rotationSpace=BCS_BoneSpace, alphaBoolBlend {blendOption=Linear}, boneToModify {boneName=laurel_2_r}}
   n10 = Get laurel_rotation_3() -> (laurel_rotation_3:double)
   n11 = "Transform (Modify) Bone - Bone: laurel_3_r"!(ComponentPose=@n9, Alpha=1, Translation="0,0,0", Rotation_Roll:float=0, Rotation_Pitch:float=0, Rotation_Yaw:float=@n10, Scale="1,1,1") -> (Pose) {rotationMode=BMM_Additive, rotationSpace=BCS_BoneSpace, alphaBoolBlend {blendOption=Linear}, boneToModify {boneName=laurel_3_r}}
   n12 = Get laurel_rotation_4() -> (laurel_rotation_4:double)
   n13 = "Transform (Modify) Bone - Bone: laurel_4_r"!(ComponentPose=@n11, Alpha=1, Translation="0,0,0", Rotation_Roll:float=0, Rotation_Pitch:float=0, Rotation_Yaw:float=@n12, Scale="1,1,1") -> (Pose) {rotationMode=BMM_Additive, rotationSpace=BCS_BoneSpace, alphaBoolBlend {blendOption=Linear}, boneToModify {boneName=laurel_4_r}}
   n14 = Get laurel_rotation_end() -> (laurel_rotation_end:double)
   n15 = "Transform (Modify) Bone - Bone: laurel_5_r"!(ComponentPose=@n13, Alpha=1, Translation="0,0,0", Rotation_Roll:float=0, Rotation_Pitch:float=0, Rotation_Yaw:float=@n14, Scale="1,1,1") -> (Pose) {rotationMode=BMM_Additive, rotationSpace=BCS_BoneSpace, alphaBoolBlend {blendOption=Linear}, boneToModify {boneName=laurel_5_r}}
   n16 = "Transform (Modify) Bone - Bone: laurel_1_l"!(ComponentPose=@n15, Alpha=1, Translation="0,0,0", Rotation_Roll:float=0, Rotation_Pitch:float=0, Rotation_Yaw:float=@n6, Scale="1,1,1") -> (Pose) {rotationMode=BMM_Additive, rotationSpace=BCS_BoneSpace, alphaBoolBlend {blendOption=Linear}, boneToModify {boneName=laurel_1_l}}
   n17 = "Transform (Modify) Bone - Bone: laurel_2_l"!(ComponentPose=@n16, Alpha=1, Translation="0,0,0", Rotation_Roll:float=0, Rotation_Pitch:float=0, Rotation_Yaw:float=@n8, Scale="1,1,1") -> (Pose) {rotationMode=BMM_Additive, rotationSpace=BCS_BoneSpace, alphaBoolBlend {blendOption=Linear}, boneToModify {boneName=laurel_2_l}}
   n18 = "Transform (Modify) Bone - Bone: laurel_3_l"!(ComponentPose=@n17, Alpha=1, Translation="0,0,0", Rotation_Roll:float=0, Rotation_Pitch:float=0, Rotation_Yaw:float=@n10, Scale="1,1,1") -> (Pose) {rotationMode=BMM_Additive, rotationSpace=BCS_BoneSpace, alphaBoolBlend {blendOption=Linear}, boneToModify {boneName=laurel_3_l}}
   n19 = "Transform (Modify) Bone - Bone: laurel_4_l"!(ComponentPose=@n18, Alpha=1, Translation="0,0,0", Rotation_Roll:float=0, Rotation_Pitch:float=0, Rotation_Yaw:float=@n12, Scale="1,1,1") -> (Pose) {rotationMode=BMM_Additive, rotationSpace=BCS_BoneSpace, alphaBoolBlend {blendOption=Linear}, boneToModify {boneName=laurel_4_l}}
   n20 = "Transform (Modify) Bone - Bone: laurel_5_l"!(ComponentPose=@n19, Alpha=1, Translation="0,0,0", Rotation_Roll:float=0, Rotation_Pitch:float=0, Rotation_Yaw:float=@n14, Scale="1,1,1") -> (Pose) {rotationMode=BMM_Additive, rotationSpace=BCS_BoneSpace, alphaBoolBlend {blendOption=Linear}, boneToModify {boneName=laurel_5_l}}
   n21 = Get SkyDiving_correction() -> (SkyDiving_correction:double)
   n22 = "Transform (Modify) Bone - Bone: laures"(ComponentPose=@n20, Alpha=@n21, Translation="0,-4,0", Rotation="0,0,15", Scale="1,1,1") {
    rotationMode=BMM_Additive
    translationMode=BMM_Additive
    alphaBoolBlend {blendOption=Linear}
    boneToModify {boneName=laures}
    rotation {roll=15}
    translation {y=-4}
   }
   n23 = "Transform (Modify) Bone - Bone: pack"(ComponentPose=@n22, Alpha=@n21, Translation="0,-2,0", Rotation="0,0,15") {
    rotationMode=BMM_Additive
    translationMode=BMM_Additive
    alphaBoolBlend {blendOption=Linear}
    boneToModify {boneName=pack}
    rotation {roll=15}
    translation {y=-2}
   }
   n24 = "Transform (Modify) Bone - Bone: pack"(ComponentPose=@n23, Translation="0,0,0", Rotation="0,0,0", Scale="0.9,0.9,0.9") {scaleMode=BMM_Additive, alphaBoolBlend {blendOption=Linear}, boneToModify {boneName=pack}, scale=<0.9 0.9 0.9>}
   Output Pose(Result=Component To Local(ComponentPose=@n24)) {layerGroup=None}
  ]}
 EventGraph {nodes [
   Event BlueprintUpdateAnimation()
   n1 = GetGameTimeInSeconds()
   n2 = float * float(A="Sin (Radians)"(A=float * float(A=1.5, B=@n1)), B=0.5)
   > Set laurels_movement(laurels_movement:Vector=MakeVector(Y=0, Z=float + float(A=@n2, B=-0.7))) -> (Output_Get:Vector)
   n3 = float + float(A=float * float(A=@n1, B=1.5), B=3)
   > Set laurel_rotation_1(laurel_rotation_1:double=float * float(A="Sin (Radians)"(A=@n3), B=3)) -> (Output_Get:double)
   n4 = float + float(A=float * float(A=@n1, B=1.5), B=2)
   > Set laurel_rotation_2(laurel_rotation_2:double=float * float(A="Sin (Radians)"(A=@n4), B=3)) -> (Output_Get:double)
   n5 = float + float(A=float * float(A=@n1, B=1.5), B=1)
   > Set laurel_rotation_3(laurel_rotation_3:double=float * float(A="Sin (Radians)"(A=@n5), B=3)) -> (Output_Get:double)
   n6 = float + float(A=float * float(A=@n1, B=1.5))
   > Set laurel_rotation_4(laurel_rotation_4:double=float * float(A="Sin (Radians)"(A=@n6), B=3)) -> (Output_Get:double)
   n7 = float + float(A=float * float(A=@n1, B=1), B=20)
   > Set laurel_rotation_end(laurel_rotation_end:double=float * float(A="Sin (Radians)"(A=@n7), B=3)) -> (Output_Get:double)
   > Set SkyDiving_correction(SkyDiving_correction:double=SelectFloat(A=1, B=0, bPickA=Get bIsSkyDiving())) -> (Output_Get:double)
   n8 = GetOwnerSkeletalMeshComponent()
   n9 = BreakRotator(InRot=BreakTransform(InTransform=GetDeltaTransformFromRefPose(self=@n8, BoneName=spine_05, BaseName=spine_01)).Rotation)
   > Set backpack_translation(backpack_translation=MakeVector(Y=MapRangeClamped(Value=@n9.Yaw, InRangeA=0, InRangeB=-30, OutRangeB=-1.3)))
   > Set backpack_rotate(backpack_rotate=MakeRotator(Roll=MapRangeClamped(Value=@n9.Yaw, InRangeA=0, InRangeB=-30, OutRangeB=-5)))
   n10 = Event OnBeginJumping()
   Set JumpingAlpha(execute=[@n10.then, @n12.then], JumpingAlpha:double=1) -> (Output_Get:double)
   n11 = Event OnEndJumping()
   Set JumpingAlpha(execute=[@n11.then, @n13.then], JumpingAlpha=0)
   n12 = Event OnBeginFalling()
   n13 = Event OnEndFalling()
   Event OnCharacterCustomizationLoaded()
   > n14 = Cast To FortPlayerPawn(Object=TryGetPawnOwner())
   n15 = GetDisplayName(Object=GetSkeletalMeshForPartType(self=@n14))
   > Set isCarrotCakeAlpha(isCarrotCakeAlpha:double="To Float (Boolean)"(InBool=MatchesWildcard(SourceString=@n15, Wildcard=*M_MED_CarrotCake*, SearchCase=CaseSensitive))) -> (Output_Get:double)
  ]}
]
variables [
 backpack_rotate {type=Rotator}
 backpack_translation {type=Vector}
 isCarrotCakeAlpha {type=number}
 JumpingAlpha {type=number}
 laurel_rotation_1 {type=number}
 laurel_rotation_2 {type=number}
 laurel_rotation_3 {type=number}
 laurel_rotation_4 {type=number}
 laurel_rotation_end {type=number}
 laurel_rotation_w_1 {type=number}
 laurel_rotation_w_2 {type=number}
 laurel_rotation_w_3 {type=number}
 laurel_rotation_w_4 {type=number}
 laurel_rotation_w_5 {type=number}
 laurels_movement {type=Vector}
 SkyDiving_correction {type=number}
]
